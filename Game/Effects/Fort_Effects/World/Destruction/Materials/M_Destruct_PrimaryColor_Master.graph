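# Master material for destruction effects on mesh particles, using the Default Lit shading model.
# A ParticleColor node provides per-particle color, alongside scalar parameters for Metallic and Roughness.
# An EmissiveColor vector parameter feeds its RGB through the shared MF_FX_EmissiveAdjustments function.
# The graph samples no textures, so the look comes from particle color and parameters.
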Material /Game/Effects/Fort_Effects/World/Destruction/Materials/M_Destruct_PrimaryColor_Master
graph {
 properties {bUsedWithMeshParticles=true, pixelDepthOffsetMode=PDOM_Legacy, usedShadingModels=Default Lit}
 nodes [
  MaterialFunctionCall(Emissive=VectorParameter() {parameterName=EmissiveColor}.RGB) {materialFunction=&MF_FX_EmissiveAdjustments}
  ParticleColor()
  ScalarParameter() {defaultValue=0.5, parameterName=Metallic}
  ScalarParameter() {defaultValue=0.5, parameterName=Roughness}
 ]
}
